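# Physics asset for the eye rig of the SunnySquawk pickaxe cosmetic, apparently for loose, physics-driven eye motion.
# Two mirrored capsule bodies, dyn_eye_A and dyn_eye_B, represent the eyes, alongside a much smaller capsule on dyn_eye_base.
# Each eye has its own constraint to the base with matching swing and twist limit angles, and no constraint joins the two eyes directly.
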PhysicsAsset /BRCosmetics/Weapons/FORT_Melee/Pickaxe_SunnySquawk/Meshes/SunnySquawk_Axe_EyesRot_Physics
bodies {
 dyn_eye_A [{shapeType=Capsule, length=2.5, radius=3, shapeName=dyn_eye_A_capsule, center=<0.0349 0.0002 -0.0321>, rotation=<rot 0 -89.9899 -72.5354>}]
 dyn_eye_B [{shapeType=Capsule, length=2.5, radius=3, shapeName=dyn_eye_B_capsule, center=<-0.0349 0.0002 -0.0321>, rotation=<rot 0 89.9899 -72.5354>}]
 dyn_eye_base [{shapeType=Capsule, length=0.505, radius=0.505, shapeName=dyn_eye_base_capsule, rotation=<rot 0 -90 0>}]
}
constraints [
 {bone1Name=dyn_eye_A, bone2Name=dyn_eye_base, swing1LimitDegrees=45, swing2LimitDegrees=45, twistLimitDegrees=45}
 {bone1Name=dyn_eye_B, bone2Name=dyn_eye_base, swing1LimitDegrees=45, swing2LimitDegrees=45, twistLimitDegrees=45}
]
