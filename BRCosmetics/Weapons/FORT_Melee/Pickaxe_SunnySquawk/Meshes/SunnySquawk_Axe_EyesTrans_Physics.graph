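# Physics asset for the eyes mesh of the SunnySquawk pickaxe, a melee weapon cosmetic.
# It gives the dyn_eye_base bone a small capsule body and each of mstr_eye_A and mstr_eye_B a larger, offset capsule.
# Two constraints join the eye bodies to the base in a star layout, both with swing and twist limit angles set to zero.
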PhysicsAsset /BRCosmetics/Weapons/FORT_Melee/Pickaxe_SunnySquawk/Meshes/SunnySquawk_Axe_EyesTrans_Physics
bodies {
 dyn_eye_base [{shapeType=Capsule, length=0.505, radius=0.505, shapeName=dyn_eye_base_capsule, rotation=<rot 0 -90 0>}]
 mstr_eye_A [{shapeType=Capsule, length=2.5, radius=3, shapeName=mstr_eye_A_capsule, center=<5 0 0>}]
 mstr_eye_B [{shapeType=Capsule, length=2.5, radius=3, shapeName=mstr_eye_B_capsule, center=<-3.7059 0.0544 -0.6192>, rotation=<rot 10.0032 -90 -9.6072>}]
}
constraints [
 {bone1Name=mstr_eye_A, bone2Name=dyn_eye_base, swing1LimitDegrees=0, swing2LimitDegrees=0, twistLimitDegrees=0}
 {bone1Name=mstr_eye_B, bone2Name=dyn_eye_base, swing1LimitDegrees=0, swing2LimitDegrees=0, twistLimitDegrees=0}
]
